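# Niagara wheel cosmetic effect for the TumbaCasa wheels, drawing flickering flares and sparks at each vehicle wheel.
# Four looping emitters partition particles by wheel, then follow each wheel through skeletal mesh sampling and VC_WheelXform.
# Flare_Core and Flare_Glow pulse via waveform size scaling, while Sparks and Sparks_Long burst omnidirectionally, the latter dragged and speed-stretched.
# Each emitter sets its own HDR painted-variant palette through VC_PaintedVariantsOverride.
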NiagaraSystem /VehicleCosmetics/Wheels/TumbaCasa/FX/NS_WheelFX_TumbaCasa
dependencies {
 usedDataInterfaces [
  &NiagaraDataInterfaceCurve
  &NiagaraDataInterfaceSpriteRendererInfo
  &NiagaraDataInterfaceSkeletalMesh
 ]
 usedDynamicInputs []
 usedModules [
  &self:SystemSpawnScript.SystemScriptSource.SystemScriptGraph.NiagaraNodeAssignment_0.SetVariables_#1
  &FN_UnpackVehicleFlags
  /VehicleCosmetics/FX/Niagara/Modules/WheelFX/VC_WheelFX_SystemSetup
  /VehicleCosmetics/FX/Niagara/Modules/VC_SetVehicleSpeedAttribs
  &SystemState
 ]
}
emitters {
 Flare_Core {
  emitterSpawnScript [/VehicleCosmetics/FX/Niagara/Modules/WheelFX/VC_WheelFX_EmitterSetup()]
  emitterUpdateScript [
   &EmitterState(Life Cycle Mode=::NewEnumerator1, Loop Behavior=::NewEnumerator1, Loop Duration Mode=::NewEnumerator1, UseLoopDelay=false)
   &SpawnRate(SpawnRate=20, Use Spawn Probability=false) {module=SpawnRate001}
   &SpawnBurst_Instantaneous(Spawn Count=$Emitter.NumWheels, Use Spawn Probability=false, UseLoopCountLimit=false) {enabled=false}
   &SpawnRate(SpawnRate=15, Use Spawn Probability=false) {enabled=false}
  ]
  particleSpawnScript [
   &PartitionParticles(Number Of Partitions=$Emitter.NumWheels, Partition Mode:ENiagara_PartitionMode_v2=::NewEnumerator5)
   /VehicleCosmetics/FX/Niagara/Modules/WheelFX/VC_WheelFX_ParticleSetup(WheelID:NiagaraInt32=$Particles.PartitionParticles.Partition)
   &VC_PaintedVariantsOverride(
    Black=<rgba 0.4 0.4 0.5 1>
    BurntSienna=<rgba 2 0.502 0.1 1>
    Cobalt=<rgba 0.5 1.5 5 1>
    Crimson=<rgba 3 0.3 0.3732 1>
    DefaultColor=<rgba 3 2.1562 1.2 1>
    ForestGreen=<rgba 0 1.5 0.0758 1>
    Lime=<rgba 0.5 1 0 1>
    Orange=<rgba 1.5 0.5232 0 1>
    Pink=<rgba 2 0.6 2 1>
    Purple=<rgba 0.1 0 0.3 2>
    Saffron=<rgba 0.9 1 0 1>
    SkyBlue=<rgba 0 0.6654 1 1>
    TitaniumWhite=<rgba 1 1 1 1>
   )
   &InitializeParticle(
    Color=$StackContext.Color
    Lifetime Max=1
    Lifetime Min=0.75
    Lifetime Mode=::NewEnumerator1
    Position=<0 20 3>
    Position Mode=::NewEnumerator1
    Sprite Rotation Mode=::NewEnumerator2
    Sprite Size Max=<15 15>
    Sprite Size Min=<8 8>
    Sprite Size Mode=::NewEnumerator2
    UsePositionOffset=false
    Write Material Random=true
   )
   &ShapeLocation(
    MaskSpawnGroup=false
    Override Seed=false
    Ring Radius=&Multiply_Float(A=$Particles.WheelRadius, B=0.35)
    Rotation Angle=90
    Rotation Angle Type:ENiagara_AngleInput=::NewEnumerator0
    Rotation Axis=<0 1 0>
    Rotation Mode=::NewEnumerator1
    Shape Primitive=::NewEnumerator4
    UseOwnerScale=true
   ) {enabled=false}
   &AddVelocity(
    Cone Angle=30
    Cone Axis=<0 1 0>
    Module.UseConeFalloff=false
    Module.UseVelDistribution=false
    Override Seed=false
    Rotation Coordinate Space=::Simulation
    Velocity Mode=::NewEnumerator2
    Velocity Speed=76
   ) {enabled=false}
   /SetVariables(Particles.glow_color:LinearColor=&Multiply_LinearColorByFloat(Linear Color=$Particles.Color, Float=0.5)) {isSetParameters=true}
   /SetVariables(Particles.glow_size:Vector2f=&Multiply_Vector2D_ByFloat(Float=1, Vector2D=$Particles.SpriteSize)) {isSetParameters=true, module=SetVariables_#2}
  ]
  particleUpdateScript [
   &ParticleState()
   &ScaleSpriteSize(
    Scale Sprite Size Mode=::NewEnumerator0
    Uniform Scale Factor=&Waveform("Waveform [1]"=::NewEnumerator9, "[1] Phase Scale"=50, "[1] Frequency"=1, "[1] Amplitude Min/Max"=<0.3 1>)
   )
   &SolveForcesAndVelocity(Clamp Velocity=false, Limit Acceleration=false)
   &SkeletalMeshLocation(
    Bone / Socket Index=$Particles.WheelID
    Bone / Socket Sampling Mode=::NewEnumerator5
    MaskSpawnGroup=false
    Mesh Sampling Type=::NewEnumerator6
    Orientation Sampling=::NewEnumerator1
    Position Sampling=::NewEnumerator2
    Skeletal Mesh=$System.SkeletalMesh
   )
   /VehicleCosmetics/FX/Niagara/Modules/WheelFX/VC_WheelXform(
    SKM Position:NiagaraPosition=$Particles.SkeletalMeshLocation.SampledPosition
    SKM Rotation:Quat4f=$Particles.SkeletalMeshLocation.SampledRotation
    SKM Velocity:Vector3f=$Particles.SkeletalMeshLocation.SampledVelocity
   )
   &SubUVAnimation(
    Random Start Frame=true
    "Sprite Renderer "=DataInterface() {propertyValues {SpriteRenderer=&self:Flare_Core_0.NiagaraSpriteRendererProperties_0}}
    SubUV Lookup Index:NiagaraFloat=&Multiply_Float(A=$Particles.Age, B=&RandomRangeFloat(Minimum=3, Maximum=5))
    UseEndFrame=false
    UseStartFrame=false
   )
  ]
  rendererClasses [&NiagaraMeshRendererProperties, &NiagaraSpriteRendererProperties]
  renderers [
   {rendererClass=&NiagaraMeshRendererProperties, rendererIndex=0}
   {rendererClass=&NiagaraSpriteRendererProperties, rendererIndex=1}
   {rendererClass=&NiagaraSpriteRendererProperties, rendererIndex=2}
  ]
 }
 Flare_Glow {
  emitterSpawnScript [&VC_WheelFX_EmitterSetup()]
  emitterUpdateScript [
   &EmitterState(Life Cycle Mode=::NewEnumerator1, Loop Behavior=::NewEnumerator1, Loop Duration Mode=::NewEnumerator1, UseLoopDelay=false)
   &SpawnRate(SpawnRate=10, Use Spawn Probability=false) {module=SpawnRate001}
   &SpawnBurst_Instantaneous(Spawn Count=$Emitter.NumWheels, Use Spawn Probability=false, UseLoopCountLimit=false)
   &SpawnRate(SpawnRate=15, Use Spawn Probability=false) {enabled=false}
  ]
  particleSpawnScript [
   &PartitionParticles(Number Of Partitions=$Emitter.NumWheels, Partition Mode:ENiagara_PartitionMode_v2=::NewEnumerator5)
   &VC_WheelFX_ParticleSetup(WheelID:NiagaraInt32=$Particles.PartitionParticles.Partition)
   &VC_PaintedVariantsOverride(
    Black=<rgba 0.5 0.5 0.5 1>
    BurntSienna=<rgba 0.75 0.1587 0 1>
    Cobalt=<rgba 0.5 1.5 5 1>
    Crimson=<rgba 5 0 0.1355 1>
    DefaultAlpha=0.5
    DefaultColor=<rgba 1 0.6719 0.3 1>
    ForestGreen=<rgba 0 1 0.0505 1>
    Lime=<rgba 0.75 1.5 0 1>
    Orange=<rgba 1.5 0.5232 0 1>
    Pink=<rgba 2 0.6 2 1>
    Purple=<rgba 0.2 0 0.68 0.5>
    SkyBlue=<rgba 0.08 0.2929 0.4 2>
    TitaniumWhite=<rgba 1 1 1 1>
   )
   &InitializeParticle(
    Color=$StackContext.Color
    Lifetime Max=5
    Lifetime Min=3
    Lifetime Mode=::NewEnumerator1
    Position=<0 20 3>
    Position Mode=::NewEnumerator1
    Sprite Rotation Mode=::NewEnumerator2
    Sprite Size Mode=::NewEnumerator3
    Uniform Sprite Size=5
    UsePositionOffset=false
    Write Material Random=true
   )
   &ShapeLocation(
    MaskSpawnGroup=false
    Override Seed=false
    Ring Radius=&Multiply_Float(A=$Particles.WheelRadius, B=0.35)
    Rotation Angle=90
    Rotation Angle Type:ENiagara_AngleInput=::NewEnumerator0
    Rotation Axis=<0 1 0>
    Rotation Mode=::NewEnumerator1
    Shape Primitive=::NewEnumerator4
    UseOwnerScale=true
   ) {enabled=false}
   &AddVelocity(
    Cone Angle=30
    Cone Axis=<0 1 0>
    Module.UseConeFalloff=false
    Module.UseVelDistribution=false
    Override Seed=false
    Rotation Coordinate Space=::Simulation
    Velocity Mode=::NewEnumerator2
    Velocity Speed=76
   ) {enabled=false}
  ]
  particleUpdateScript [
   &ParticleState()
   &ScaleSpriteSize(
    Scale Sprite Size Mode=::NewEnumerator0
    Uniform Scale Factor=&Waveform("Waveform [1]"=::NewEnumerator9, "[1] Phase Scale"=35, "[1] Frequency"=1, "[1] Amplitude Min/Max"=<0.25 1>)
   )
   &SolveForcesAndVelocity(Clamp Velocity=false, Limit Acceleration=false)
   &SkeletalMeshLocation(
    Bone / Socket Index=$Particles.WheelID
    Bone / Socket Sampling Mode=::NewEnumerator5
    MaskSpawnGroup=false
    Mesh Sampling Type=::NewEnumerator6
    Orientation Sampling=::NewEnumerator1
    Position Sampling=::NewEnumerator2
    Skeletal Mesh=$System.SkeletalMesh
   )
   &VC_WheelXform(
    SKM Position:NiagaraPosition=$Particles.SkeletalMeshLocation.SampledPosition
    SKM Rotation:Quat4f=$Particles.SkeletalMeshLocation.SampledRotation
    SKM Velocity:Vector3f=$Particles.SkeletalMeshLocation.SampledVelocity
   )
  ]
  rendererClasses [&NiagaraMeshRendererProperties, &NiagaraSpriteRendererProperties]
  renderers [
   {rendererClass=&NiagaraMeshRendererProperties, rendererIndex=0}
   {rendererClass=&NiagaraSpriteRendererProperties, rendererIndex=1}
  ]
 }
 Sparks {
  emitterSpawnScript [&VC_WheelFX_EmitterSetup()]
  emitterUpdateScript [
   &EmitterState(Life Cycle Mode=::NewEnumerator1, Loop Behavior=::NewEnumerator1, Loop Duration Mode=::NewEnumerator1, UseLoopDelay=false)
   &SpawnRate(SpawnRate=50, Use Spawn Probability=false) {module=SpawnRate001}
   &SpawnBurst_Instantaneous(Spawn Count=$Emitter.NumWheels, Use Spawn Probability=false, UseLoopCountLimit=false) {enabled=false}
   &SpawnRate(SpawnRate=15, Use Spawn Probability=false) {enabled=false}
  ]
  particleSpawnScript [
   &PartitionParticles(Number Of Partitions=$Emitter.NumWheels, Partition Mode:ENiagara_PartitionMode_v2=::NewEnumerator5)
   &VC_WheelFX_ParticleSetup(WheelID:NiagaraInt32=$Particles.PartitionParticles.Partition)
   &VC_PaintedVariantsOverride(
    Black=<rgba 0.5 0.5 0.5 1>
    BurntSienna=<rgba 2 0.4231 0 1>
    Cobalt=<rgba 0.2 0.6 2 1>
    Crimson=<rgba 5 0.25 0.3787 1>
    DefaultColor=<rgba 3 2.0156 0.9 1>
    ForestGreen=<rgba 0 1 0.0505 1>
    Grey=<rgba 1 1 1 1>
    Lime=<rgba 1 2 0 1>
    Orange=<rgba 1.5 0.5232 0 1>
    Pink=<rgba 2 0.6 2 1>
    Purple=<rgba 0.4167 0 1 2>
    Saffron=<rgba 0.9 1 0 1>
    SkyBlue=<rgba 0 0.6654 1 1>
    TitaniumWhite=<rgba 1 1 1 1>
   )
   &InitializeParticle(
    Color=$StackContext.Color
    Lifetime Max=0.125
    Lifetime Min=0.075
    Lifetime Mode=::NewEnumerator1
    Position=<0 20 3>
    Position Mode=::NewEnumerator1
    Sprite Rotation Mode=::NewEnumerator2
    Sprite Size Mode=::NewEnumerator1
    Uniform Sprite Size Max=4
    Uniform Sprite Size Min=1.5
    UsePositionOffset=false
    Write Material Random=true
   )
   &ShapeLocation(
    MaskSpawnGroup=false
    Override Seed=false
    Ring Radius=&Multiply_Float(A=$Particles.WheelRadius, B=0.01)
    Rotation Angle=90
    Rotation Angle Type:ENiagara_AngleInput=::NewEnumerator0
    Rotation Axis=<0 1 0>
    Rotation Mode=::NewEnumerator1
    Shape Primitive=::NewEnumerator4
    UseOwnerScale=true
   ) {enabled=false}
   &AddVelocity(
    Cone Angle=360
    Cone Axis=<0 0 0>
    Module.UseConeFalloff=false
    Module.UseVelDistribution=false
    Override Seed=false
    Rotation Coordinate Space=::Simulation
    Velocity Mode=::NewEnumerator2
    Velocity Speed=85
   )
   /SetVariables(Particles.glow_color:LinearColor=&Multiply_LinearColorByFloat(Linear Color=$Particles.Color, Float=0.5)) {isSetParameters=true}
  ]
  particleUpdateScript [
   &ParticleState()
   &ScaleSpriteSize(Uniform Curve Sprite Scale=DataInterface() {propertyValues {Curve {
       defaultValue=0
       keys [
        {interpMode=RCIM_Cubic, tangentMode=RCTM_User, time=0, value=0.4}
        {interpMode=RCIM_Cubic, tangentMode=RCTM_User, time=0.1653, value=1}
        {arriveTangent=0.1369, interpMode=RCIM_Cubic, leaveTangent=0.1369, tangentMode=RCTM_User, time=0.496, value=1}
        {tangentMode=RCTM_User, time=1, value=0.4}
       ]
      }}})
   &SolveForcesAndVelocity(Clamp Velocity=false, Limit Acceleration=false)
   &SkeletalMeshLocation(
    Bone / Socket Index=$Particles.WheelID
    Bone / Socket Sampling Mode=::NewEnumerator5
    MaskSpawnGroup=false
    Mesh Sampling Type=::NewEnumerator6
    Orientation Sampling=::NewEnumerator1
    Position Sampling=::NewEnumerator2
    Skeletal Mesh=$System.SkeletalMesh
   )
   &VC_WheelXform(
    SKM Position:NiagaraPosition=$Particles.SkeletalMeshLocation.SampledPosition
    SKM Rotation:Quat4f=$Particles.SkeletalMeshLocation.SampledRotation
    SKM Velocity:Vector3f=$Particles.SkeletalMeshLocation.SampledVelocity
   )
   &SubUVAnimation(
    Random Start Frame=true
    "Sprite Renderer "=DataInterface() {propertyValues {SpriteRenderer=&self:Sparks_0.NiagaraSpriteRendererProperties_0}}
    SubUV Lookup Index:NiagaraFloat=&Multiply_Float(A=$Particles.Age, B=&RandomRangeFloat(Minimum=1, Maximum=2))
    UseEndFrame=false
    UseStartFrame=false
   )
   /SetVariables(Particles.glow_size:Vector2f=&Multiply_Vector2D_ByFloat(Float=2, Vector2D=$Particles.SpriteSize)) {isSetParameters=true, module=SetVariables_#2}
  ]
  rendererClasses [&NiagaraMeshRendererProperties, &NiagaraSpriteRendererProperties]
  renderers [
   {rendererClass=&NiagaraMeshRendererProperties, rendererIndex=0}
   {rendererClass=&NiagaraSpriteRendererProperties, rendererIndex=1}
   {rendererClass=&NiagaraSpriteRendererProperties, rendererIndex=2}
  ]
 }
 Sparks_Long {
  emitterSpawnScript [&VC_WheelFX_EmitterSetup()]
  emitterUpdateScript [
   &EmitterState(Life Cycle Mode=::NewEnumerator1, Loop Behavior=::NewEnumerator1, Loop Duration Mode=::NewEnumerator1, UseLoopDelay=false)
   &SpawnRate(SpawnRate=100, Use Spawn Probability=false) {module=SpawnRate001}
   &SpawnBurst_Instantaneous(Spawn Count=$Emitter.NumWheels, Use Spawn Probability=false, UseLoopCountLimit=false) {enabled=false}
   &SpawnRate(SpawnRate=15, Use Spawn Probability=false) {enabled=false}
  ]
  particleSpawnScript [
   &PartitionParticles(Number Of Partitions=$Emitter.NumWheels, Partition Mode:ENiagara_PartitionMode_v2=::NewEnumerator5)
   &VC_WheelFX_ParticleSetup(WheelID:NiagaraInt32=$Particles.PartitionParticles.Partition)
   &VC_PaintedVariantsOverride(
    Black=<rgba 1 1 1 1>
    BurntSienna=<rgba 4 0.8463 0 1>
    Cobalt=<rgba 0.5 1.5 5 1>
    Crimson=<rgba 5 0 0.1355 1>
    DefaultColor=<rgba 5 3.3594 1.5 1>
    ForestGreen=<rgba 0 1 0.0505 1>
    Grey=<rgba 2 2 2 1>
    Lime=<rgba 0.75 1.5 0 1>
    Orange=<rgba 1.5 0.5232 0 1>
    Pink=<rgba 3 0.9 3 1>
    Purple=<rgba 0.8208 0 2 1>
    Saffron=<rgba 1.8 2 0 1>
    SkyBlue=<rgba 0 1.9962 3 1>
    TitaniumWhite=<rgba 2 2 2 1>
   )
   &InitializeParticle(
    Color=$StackContext.Color
    Lifetime Max=0.225
    Lifetime Min=0.075
    Lifetime Mode=::NewEnumerator1
    Position=<0 20 3>
    Position Mode=::NewEnumerator1
    Sprite Rotation Mode=::NewEnumerator2
    Sprite Size Mode=::NewEnumerator1
    Uniform Sprite Size Max=0.3
    Uniform Sprite Size Min=0.15
    UsePositionOffset=false
    Write Material Random=true
   )
   &ShapeLocation(
    MaskSpawnGroup=false
    Override Seed=false
    Ring Radius=&Multiply_Float(A=$Particles.WheelRadius, B=0.025)
    Rotation Angle=90
    Rotation Angle Type:ENiagara_AngleInput=::NewEnumerator0
    Rotation Axis=<0 1 0>
    Rotation Mode=::NewEnumerator1
    Shape Primitive=::NewEnumerator4
    UseOwnerScale=true
   ) {enabled=false}
   &AddVelocity(
    Cone Angle=360
    Cone Axis=<0 0 0>
    Module.UseConeFalloff=false
    Module.UseVelDistribution=false
    Override Seed=false
    Rotation Coordinate Space=::Simulation
    Velocity Mode=::NewEnumerator2
    Velocity Speed=&RandomRangeFloat(Minimum=50, Maximum=100)
   )
   /SetVariables(Particles.glow_color:LinearColor=&Multiply_LinearColorByFloat(Linear Color=$Particles.Color, Float=0.3)) {isSetParameters=true}
  ]
  particleUpdateScript [
   &ParticleState()
   &ScaleSpriteSize(Uniform Curve Sprite Scale=DataInterface() {propertyValues {Curve {
       defaultValue=0
       keys [
        {interpMode=RCIM_Cubic, tangentMode=RCTM_User, time=0, value=0.4}
        {interpMode=RCIM_Cubic, tangentMode=RCTM_User, time=0.1653, value=1}
        {arriveTangent=0.1369, interpMode=RCIM_Cubic, leaveTangent=0.1369, tangentMode=RCTM_User, time=0.496, value=1}
        {tangentMode=RCTM_User, time=1, value=0.4}
       ]
      }}})
   &Drag(Use Linear Drag=true, Use Rotational Drag=true)
   &SolveForcesAndVelocity(Clamp Velocity=false, Limit Acceleration=false)
   &ScaleSpriteSizeBySpeed(Max Scale Factor=<1 10>, Min Scale Factor=<1 0.5>, Velocity Threshold=2)
   &SkeletalMeshLocation(
    Bone / Socket Index=$Particles.WheelID
    Bone / Socket Sampling Mode=::NewEnumerator5
    MaskSpawnGroup=false
    Mesh Sampling Type=::NewEnumerator6
    Orientation Sampling=::NewEnumerator1
    Position Sampling=::NewEnumerator2
    Skeletal Mesh=$System.SkeletalMesh
   )
   &VC_WheelXform(
    SKM Position:NiagaraPosition=$Particles.SkeletalMeshLocation.SampledPosition
    SKM Rotation:Quat4f=$Particles.SkeletalMeshLocation.SampledRotation
    SKM Velocity:Vector3f=$Particles.SkeletalMeshLocation.SampledVelocity
   )
   /SetVariables(Particles.glow_Size:Vector2f=&Multiply_Vector2D_ByFloat(Float=5, Vector2D=$Particles.SpriteSize)) {isSetParameters=true, module=SetVariables_#2}
  ]
  rendererClasses [&NiagaraMeshRendererProperties, &NiagaraSpriteRendererProperties]
  renderers [
   {rendererClass=&NiagaraMeshRendererProperties, rendererIndex=0}
   {rendererClass=&NiagaraSpriteRendererProperties, rendererIndex=1}
   {rendererClass=&NiagaraSpriteRendererProperties, rendererIndex=2}
  ]
 }
}
summary {
 systemName=NS_WheelFX_TumbaCasa
 emitters [
  {emitterName=Flare_Core, rendererClasses [&NiagaraMeshRendererProperties, &NiagaraSpriteRendererProperties]}
  {emitterName=Sparks, rendererClasses [&NiagaraMeshRendererProperties, &NiagaraSpriteRendererProperties]}
  {emitterName=Sparks_Long, rendererClasses [&NiagaraMeshRendererProperties, &NiagaraSpriteRendererProperties]}
  {emitterName=Flare_Glow, rendererClasses [&NiagaraMeshRendererProperties, &NiagaraSpriteRendererProperties]}
 ]
 userVariables [
  User.bIsFrontendPreview {type=NiagaraBool}
  User.bIsPainted {type=NiagaraBool}
  User.ColorVariantID {type=NiagaraInt32, defaultValue=15}
  User.PaintedColor {type=LinearColor, description="", defaultValue {struct=LinearColor, value=<rgba 1 1 1 1>}}
  User.PaintedScalar {type=NiagaraFloat, defaultValue=1, description=""}
  User.VehicleFlags {type=NiagaraInt32, defaultValue=806289408}
  User.Wheel_Radius_Array {
   type=NiagaraDataInterfaceArrayFloat
   defaultValue {
    struct=NiagaraExt_VariableValue_DataInterface
    value {
     dataInterface=&self:NiagaraDataInterfaceArrayFloat_1
     dataInterfaceClass=&NiagaraDataInterfaceArrayFloat
    }
   }
  }
  User.Wheel_Width_Array {
   type=NiagaraDataInterfaceArrayFloat
   defaultValue {
    struct=NiagaraExt_VariableValue_DataInterface
    value {
     dataInterface=&self:NiagaraDataInterfaceArrayFloat_0
     dataInterfaceClass=&NiagaraDataInterfaceArrayFloat
    }
   }
  }
  User.WheelTouchingArray {
   type=NiagaraDataInterfaceArrayFloat
   defaultValue {
    struct=NiagaraExt_VariableValue_DataInterface
    value {
     dataInterface=&self:NiagaraDataInterfaceArrayFloat_2
     dataInterfaceClass=&NiagaraDataInterfaceArrayFloat
    }
   }
  }
 ]
}
